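# Sound cue for the Season 23 "NightNight" music pack, playing its lobby track as looping music.
# A single looping wave player feeds the output directly, routed through the music packs sound class with a slight volume boost.
# Attenuation and spatialization are off, and concurrency allows one instance, stopping the oldest.
# A submix send targets the music audio analysis submix.
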SoundCue /Game/Athena/Sounds/MusicPacks/s23_NightNight/s23_Musicpack_NightNight_Cue
root=WavePlayer_1
nodes [{id=WavePlayer_1, properties {bLooping=true, soundWaveAssetPtr=&/S23_NightNight_LobbyTrack}}]
properties {
 maxDistance=2097152
 soundClassObject=/Game/Sounds/FortSoundClassesAndModes/Fort_Music_MusicPacks
 volumeMultiplier=1.2
 attenuationOverrides {bAttenuate=false, bSpatialize=false}
 concurrencyOverrides {maxCount=1, resolutionRule=StopOldest}
 soundSubmixSends [{minSendLevel=0, soundSubmix=&Fortnite_AudioAnalysis_Music}]
}
